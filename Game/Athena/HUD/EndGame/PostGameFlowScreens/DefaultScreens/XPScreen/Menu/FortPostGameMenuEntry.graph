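# Widget Blueprint for a menu entry on the post-game XP screen in Athena's end-game flow, derived from the native FortPostGameMenuEntry class.
# Its root is a UIKit Regular button, bound to the inherited Overlay_Entry variable and given vertical padding.
# The inherited InputActionWidget slot has no widget assigned.
# The event graph is empty and no variables are added, so the Blueprint only supplies layout and styling.
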WidgetBlueprint /Game/Athena/HUD/EndGame/PostGameFlowScreens/DefaultScreens/XPScreen/Menu/FortPostGameMenuEntry
graphs [EventGraph {links [], nodes []}]
variables []
widgets {
 info {
  inheritedWidgetCount=2
  parentClass=/Script/FortniteUI.FortPostGameMenuEntry
  rootWidgetClass=/FortUILibrary/UIKit/Buttons/WBP_UIKit_Button_Regular.WBP_UIKit_Button_Regular_C
  widgetCount=2
 }
 widgets [
  Overlay_Entry {
   widgetClassPath=&WBP_UIKit_Button_Regular
   bInherited=true
   bIsVariable=true
   widget=Overlay_Entry
   properties {padding=<ltrb 0 12 0 12>}
  }
  InputActionWidget {widgetClassPath=&CommonActionWidget, bInherited=true, widget=null}
 ]
}
