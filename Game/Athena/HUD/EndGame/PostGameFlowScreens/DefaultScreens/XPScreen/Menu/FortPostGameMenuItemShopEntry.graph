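# Item Shop menu entry widget for the post-game XP screen, based on the native FortPostGameMenuItemShopEntry class.
# A NormalBangWrapper root hosts the inherited Overlay_Entry UIKit button in its named slot and uses the AthenaDirectedAcquisitionTab bang type for notification badging.
# The button overrides hover and press sounds, including locked and selected states, and its fade-out time.
# The event graph is empty, so the Blueprint adds only layout and styling.
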WidgetBlueprint /Game/Athena/HUD/EndGame/PostGameFlowScreens/DefaultScreens/XPScreen/Menu/FortPostGameMenuItemShopEntry
graphs [EventGraph {links [], nodes []}]
variables []
widgets {
 info {
  inheritedWidgetCount=2
  namedSlotCount=1
  parentClass=/Script/FortniteUI.FortPostGameMenuItemShopEntry
  rootWidgetClass=/Game/UI/Foundation/Bangs/NormalBangWrapper.NormalBangWrapper_C
  widgetCount=3
 }
 widgets [
  NormalBangWrapper {
   widgetClassPath=/Game/UI/Foundation/Bangs/NormalBangWrapper
   bIsVariable=true
   widget=NormalBangWrapper
   properties {bangType=AthenaDirectedAcquisitionTab}
  }
  Overlay_Entry {
   widgetClassPath=&WBP_UIKit_Button_Regular
   bInherited=true
   bIsVariable=true
   namedSlotHost=NormalBangWrapper
   widget=Overlay_Entry
   properties {
    fadeOutTime=0.21
    hoveredSlateSoundOverride {resourceObject=/Game/Sounds/Fort_UI_Sounds/FrontEnd/New_FrontEnd/Fortnite_UI_MainMenu_Tab_Hover_01_Cue}
    lockedHoveredSlateSoundOverride {resourceObject=&Fortnite_UI_MainMenu_Tab_Hover_01_Cue}
    lockedPressedSlateSoundOverride {resourceObject=/Game/Sounds/Fort_UI_Sounds/FrontEnd/CareerPage/Fortnite_UI_Career_Button_Generic_01_Cue}
    pressedSlateSoundOverride {resourceObject=&Fortnite_UI_Career_Button_Generic_01_Cue}
    selectedHoveredSlateSoundOverride {resourceObject=&Fortnite_UI_MainMenu_Tab_Hover_01_Cue}
    selectedPressedSlateSoundOverride {resourceObject=&Fortnite_UI_Career_Button_Generic_01_Cue}
   }
  }
  InputActionWidget {widgetClassPath=&CommonActionWidget, bInherited=true, widget=null}
 ]
}
